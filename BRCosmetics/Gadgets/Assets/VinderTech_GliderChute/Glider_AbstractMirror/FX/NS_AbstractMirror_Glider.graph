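# Niagara system for the AbstractMirror glider's flight effects: engine glows, thruster sprites, speed lines and an outer cone.
# JetEngine (sprites, mesh, light) and SpriteThrusters use spawn groups to pin particles to two skeletal mesh sockets, fading with system Timeline.
# OuterCone meshes and torus-spawned, speed-stretched SpeedLine sprites convey airflow.
# User.ForwardDot scales intensity throughout, with front-end preview overrides and a deployed check gating preview speed lines.
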NiagaraSystem /BRCosmetics/Gadgets/Assets/VinderTech_GliderChute/Glider_AbstractMirror/FX/NS_AbstractMirror_Glider
dependencies {
 usedDataInterfaces [
  &NiagaraDataInterfaceVector2DCurve
  &NiagaraDataInterfaceCurve
  &NiagaraDataInterfaceColorCurve
  &NiagaraDataInterfaceArrayFloat
 ]
 usedDynamicInputs [&SetBoolByFloatComparison]
 usedModules [&SystemState, &Timeline]
}
emitters {
 JetEngine {
  emitterUpdateScript [
   &EmitterState(
    Inactive Response=::NewEnumerator1
    Life Cycle Mode=::NewEnumerator1
    Loop Behavior=::NewEnumerator1
    Loop Duration Mode=::NewEnumerator1
    UseLoopDelay=false
   )
   &SpawnBurst_Instantaneous(Use Spawn Probability=false, UseLoopCountLimit=false)
   &SpawnBurst_Instantaneous(Spawn Group=1, Use Spawn Probability=false, UseLoopCountLimit=false) {module=SpawnBurst_Instantaneous001}
  ]
  particleSpawnScript [
   &InitializeParticle(
    Color=<rgba 1 0.7947 0.4635 0.25>
    Mesh Scale=<0.6 1.25 0.3>
    Mesh Scale Mode=::NewEnumerator4
    Sprite Size Mode=::NewEnumerator3
    Uniform Sprite Size=5
    UsePositionOffset=true
    Write Material Random=true
   )
   &SpriteFacingAndAlignment(WriteAlignment=false, WriteFacing=true)
   /SetVariables(
    Particles.GlowCameraOffset=22
    Particles.GlowColor=<rgba 0 0.2326 1 1>
    Particles.GlowPosition:NiagaraPosition=&AddVectorToPosition(Position=$Particles.Position, Vector=<-20 20 0>)
    Particles.GlowSize=<4 4>
   ) {isSetParameters=true}
   &InitialMeshOrientation(Mesh Axis To Orient=<0 1 0>, Mesh Orientation Mode=::NewEnumerator2, Orientation Vector=<0 -1 0>, Use Rotation Vector=false) {enabled=false}
   /SetVariables(Particles.InnerGlowFacing:Vector3f=<1 0 0>) {isSetParameters=true}
  ]
  particleUpdateScript [
   &ParticleState(Kill Particles When Lifetime Has Elapsed=false)
   &ScaleMeshSize(Scale Factor=&Lerp_Vector(
     A=<1 0.5 1>
     B=<1 1.25 1>
     Alpha=&ClampFloat(
      Float=$User.ForwardDot
      Min=&MakeCustomFloatFromBool(Bool=$User.bIsFrontEndPreview, True Float=1, False Float=0)
     )
    ))
   &SkeletalMeshLocation(
    Bone / Socket Index=&MaskIntBySpawnGroup(Masked Int=0, Passthrough Int=1, Spawn Group Mask=0)
    MaskSpawnGroup=false
    Mesh Sampling Type=::NewEnumerator5
    Skeletal Mesh=$User.SkeletalMesh
    Socket Sampling Mode=::NewEnumerator5
   )
   &FN_Light_Attributes(
    Color=&MultiplyLinearColor(Linear Color=$Particles.Color, Scale Factor=&MakeVector4_FromFloat(Float=3))
    Create Color Attribute=true
    Create Exponent Attribute=true
    Create Position Attribute=true
    Create Radius Attribute=true
    Exponent=8
    Position=&AddVectorToPosition(Position=$Particles.Position, Vector=<6 0 0>)
    Radius=40
    Scale Color=&Lerp_Float(A=1, B=2, Alpha=&ClampFloat(Float=$User.ForwardDot))
   )
   &ScaleColor(Scale Alpha=&Multiply_Float(A=1, B=$System.Timeline.NormalizedTime), ScaleA=true, ScaleRGB=false)
  ]
  rendererClasses [
   &NiagaraSpriteRendererProperties
   &NiagaraMeshRendererProperties
   &NiagaraLightRendererProperties
  ]
  renderers [
   {rendererClass=&NiagaraSpriteRendererProperties}
   {rendererClass=&NiagaraMeshRendererProperties, rendererIndex=1}
   {rendererClass=&NiagaraLightRendererProperties, rendererIndex=2}
   {rendererClass=&NiagaraSpriteRendererProperties, rendererIndex=3}
  ]
 }
 OuterCone {
  emitterUpdateScript [
   &EmitterState()
   &SpawnRate(
    SpawnRate=&FN_FrontEndFloatModifier(Gameplay=&Lerp_Float(A=0, B=8, Alpha=$User.ForwardDot), Front End=8)
    Use Spawn Probability=false
   )
  ]
  particleSpawnScript [
   &InitializeParticle(
    Color=&MakeLinearColorFromVectorAndFloat(
     "Vector (RGB)"=<1 1 1>
     "Float (Alpha)"=&Multiply_Float(A=&RandomRangeFloat(Minimum=0.8, Maximum=1), B=0.025)
    )
    Lifetime=0.35
    Mesh Scale=&Multiply_VectorByFloat(Vector=<1 1.5 2>, Float=1)
    Mesh Scale Mode=::NewEnumerator4
    Sprite Size Mode=::NewEnumerator3
    UsePositionOffset=false
    Write Material Random=true
   )
   &InitialMeshOrientation(
    Rotation=&FN_FrontEndVector3Modifier(Gameplay=<0 0 0.75>, Front End=<0 0 0.75>)
    Use Orientation Vector=true
    Use Rotation Vector=true
   )
  ]
  particleUpdateScript [&ParticleState()]
  rendererClasses [&NiagaraMeshRendererProperties]
  renderers [{rendererClass=&NiagaraMeshRendererProperties, rendererIndex=0}]
 }
 SpeedLine {
  emitterUpdateScript [
   &EmitterState()
   &SpawnRate(
    SpawnRate=&MakeCustomFloatFromBool(Bool=$User.bIsFrontEndPreview, True Float=50, False Float=90)
    Use Spawn Probability=true
   ) {module=SpawnRate001}
  ]
  particleSpawnScript [
   &InitializeParticle(
    Color=&MakeLinearColorFromVectorAndFloat(
     "Vector (RGB)"=<1 1 1>
     "Float (Alpha)"=&Multiply_Float(
      A=&UniformRangedFloat(Minimum=0.1, Maximum=0.18)
      B=&MakeCustomFloatFromBool(Bool=$User.bIsFrontEndPreview, True Float=0.5, False Float=1)
     )
    )
    Lifetime Max=0.16
    Lifetime Min=0.1
    Lifetime Mode=::NewEnumerator1
    Sprite Size Max=<3 10>
    Sprite Size Min=<2 2>
    Sprite Size Mode=::NewEnumerator2
    UsePositionOffset=false
    Write Material Random=true
   )
   &TorusLocation(
    Align Torus With Axis=true
    Handle Radius=0
    Large Radius:NiagaraFloat=&RandomRangeFloat(Minimum=80, Maximum=150)
    MaskSpawnGroup=false
    Module.Override Randomness=false
    Offset=<0 -180 0>
    Torus Axis=<0 1 0>
    Torus Origin:NiagaraPosition=&Add_Vector(
     A=Linked() {linkedVariable {name=Particles.Position, type=NiagaraPosition}}
     B=&MakeVector(X=0, Y=&MakeCustomFloatFromBool(Bool=$User.bIsFrontEndPreview, True Float=200, False Float=0), Z=0)
    )
   )
   &AddVelocityFromPoint(
    Module.Use Falloff Distance=false
    Origin Offset=<0 -500 0>
    Velocity Strength=&Multiply_Float(A=600, B=&MakeCustomFloatFromBool(Bool=$User.bIsFrontEndPreview, True Float=1, False Float=1))
   )
   &AddVelocity(Velocity=&Multiply_Vector(
     A=&RandomRangeVector(Minimum=<0 300 0>, Maximum=<0 700 0>)
     B=&MakeVector(X=1, Y=&MakeCustomFloatFromBool(Bool=$User.bIsFrontEndPreview, True Float=1, False Float=1), Z=1)
    ))
  ]
  particleUpdateScript [
   &SolveForcesAndVelocity(Clamp Velocity=false, Limit Acceleration=false)
   &ParticleState()
   &ScaleSpriteSize(Scale Factor=&Vector2DFromCurve(
     Vector2Curve=DataInterface() {propertyValues {XCurve {keys [<at 0 0.5 RCIM_Cubic>,<at 0.5 1 RCIM_Cubic>]}, YCurve {keys [<at 0 0 RCIM_Cubic>,<at 0.5 1 RCIM_Cubic>]}}}
     ScaleCurve=&Vector2DFromFloat(Value=1)
    ))
   &ScaleSpriteSizeBySpeed(Max Scale Factor=<1 50>, Min Scale Factor=<0.01 0.05>, Velocity Threshold=1)
   &ScaleColor(
    Scale Alpha=&FloatFromCurve(
     FloatCurve=DataInterface() {propertyValues {Curve {keys [
         <at 0 0 RCIM_Cubic>
         <at 0.3759 1.0207>
         {arriveTangent=-2.221, interpMode=RCIM_Cubic, leaveTangent=-2.221, tangentMode=RCTM_User, time=1, value=0}
        ]}}}
     Scale Curve=&MakeCustomFloatFromBool(
      Bool=$User.bIsFrontEndPreview
      True Float=&MakeCustomFloatFromBool(Bool=$User.bIsFullyDeployed, True Float=1, False Float=0)
      False Float=&Lerp_Float(A=0, B=1, Alpha=$User.ForwardDot)
     )
    )
    Scale RGB=&VectorFromFloat(Value=1)
    ScaleA=true
    ScaleRGB=true
   )
  ]
  rendererClasses [&NiagaraSpriteRendererProperties]
  renderers [{rendererClass=&NiagaraSpriteRendererProperties}]
 }
 SpriteThrusters {
  emitterUpdateScript [
   &EmitterState(Life Cycle Mode=::NewEnumerator1, Loop Duration=2, UseLoopDelay=false)
   &SpawnRate(SpawnRate=50, Use Spawn Probability=false)
   &SpawnRate(SpawnGroup=1, SpawnRate=50, Use Spawn Probability=false) {module=SpawnRate001}
  ]
  particleSpawnScript [
   &InitializeParticle(
    Color=<rgba 0.5156 0.5156 0.5156 1>
    Lifetime=0.2
    Mass Mode=::NewEnumerator3
    Sprite Rotation Mode=::NewEnumerator2
    Sprite Size Mode=::NewEnumerator1
    Uniform Sprite Size Max=4
    Uniform Sprite Size Min=3
    UsePositionOffset=false
    Write Material Random=true
   )
   &AddVelocity(
    Rotation Coordinate Space=::Simulation
    Velocity=<300 0 0>
    Velocity Speed Scale=&Lerp_Float(
     A=0.5
     B=1
     Alpha=&ClampFloat(
      Float=$User.ForwardDot
      Min=&MakeCustomFloatFromBool(Bool=$User.bIsFrontEndPreview, True Float=1, False Float=0)
     )
    )
   )
  ]
  particleUpdateScript [
   &SkeletalMeshLocation(
    Bone / Socket Index=&MaskIntBySpawnGroup(Masked Int=0, Passthrough Int=1, Spawn Group Mask=0)
    MaskSpawnGroup=false
    Mesh Sampling Type=::NewEnumerator5
    Skeletal Mesh=$User.SkeletalMesh
    Socket Sampling Mode=::NewEnumerator5
   )
   &ParticleState()
   &Drag(Drag=5, Use Linear Drag=true, Use Rotational Drag=false)
   &ScaleColor(
    Linear Color Curve=DataInterface() {propertyValues {
      AlphaCurve {keys [<at 0 1>,<at 1 0>]}
      BlueCurve {keys [<at 0 0.5008>,<at 0.3146 0.3906>]}
      GreenCurve {keys [<at 0 0>,<at 0.3146 0.2374>]}
      RedCurve {keys [<at 0 0.5781>,<at 0.3146 0>]}
     }}
    Scale Mode=::NewEnumerator2
   )
   &SolveForcesAndVelocity(Clamp Velocity=false, Limit Acceleration=false)
   &ScaleSpriteSizeBySpeed(Max Scale Factor=<1 6>, Min Scale Factor=<0.5 1>, Velocity Threshold=300)
   &ScaleColor(Scale Alpha=&Multiply_Float(A=1, B=$System.Timeline.NormalizedTime), ScaleA=true, ScaleRGB=false) {module=ScaleColor001}
  ]
  rendererClasses [&NiagaraSpriteRendererProperties]
  renderers [{rendererClass=&NiagaraSpriteRendererProperties}]
 }
}
summary {
 systemName=NS_AbstractMirror_Glider
 emitters [
  {emitterName=SpeedLine, rendererClasses [&NiagaraSpriteRendererProperties]}
  {emitterName=OuterCone, rendererClasses [&NiagaraMeshRendererProperties]}
  {
   emitterName=JetEngine
   rendererClasses [
    &NiagaraSpriteRendererProperties
    &NiagaraMeshRendererProperties
    &NiagaraLightRendererProperties
   ]
  }
  {emitterName=SpriteThrusters, rendererClasses [&NiagaraSpriteRendererProperties]}
 ]
 userVariables [
  User.bIsFrontEnd {type=NiagaraBool}
  User.bIsFrontEndPreview {type=NiagaraBool}
  User.bIsFullyDeployed {type=NiagaraBool}
  User.Delay {type=NiagaraFloat, defaultValue=0.3, description=""}
  User.ForwardDot {type=NiagaraFloat, defaultValue=1, description=""}
  User.RibbonIDOffset {type=NiagaraInt32}
  User.SkeletalMesh {
   type=NiagaraDataInterfaceSkeletalMesh
   defaultValue {
    struct=NiagaraExt_VariableValue_DataInterface
    value {dataInterface=&self:NiagaraDataInterfaceSkeletalMesh_0}
   }
  }
 ]
}
